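# Terrain material for the Artemis Milkshake set, combining baked rock maps with tiled ground that blend toward the bottom by world-height and normal masks, with HLOD offset correction.
# On the default shading path, it adds an optional height-based wet layer and height-blends into the Apollo landscape runtime virtual texture.
# Mobile skips both, using vertex-color height-blended ground textures with a debug switch.
# A SingleRock switch bypasses the baked/ground blend.
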Material /Game/Environments/Artemis/Sets/Milkshake/Terrain/Materials/M_Milkshake
graph {
 properties {bTangentSpaceNormal=false, bUseMaterialAttributes=true, pixelDepthOffsetMode=PDOM_Legacy, usedShadingModels=Default Lit}
 nodes [
  Composite() {subgraphName=Height Masks}
  n1 = VectorParameter() {parameterName=HLODOffsetLocation, defaultValue=<rgba 0 0 0 1>}
  Composite!(HLOD_Offset=@n1.RGB) {subgraphName=Height Masks}
  Composite!() {subgraphName=Wet Height}
  n2 = Divide(
   A=ComponentMask(None=WorldPosition().XYZ) {g=true, r=true}
   B=Multiply(A=ScalarParameter() {defaultValue=512, group=Ground Mobile, parameterName=Ground Tile Mobile}) {constB=-1}
  )
  n3 = TextureSampleParameter2D(UVs=@n2) {
   group=Ground Mobile
   parameterName=Ground Spec Mobile
   samplerType=SAMPLERTYPE_Masks
   sortPriority=3
   texture=/Game/Environments/Apollo/Sets/Terrain/Textures/T_Apollo_Grass_S_01
  }
  n4 = TextureSampleParameter2D(UVs=@n2) {
   group=Ground Mobile
   parameterName=Ground Spec Mobile 2
   samplerType=SAMPLERTYPE_Masks
   sortPriority=3
   texture=&T_Apollo_Grass_S_01
  }
  n5 = TextureSampleParameter2D(UVs=@n2) {
   group=Ground Mobile
   parameterName=Ground Diffuse Mobiile
   sortPriority=2
   texture=/Game/Environments/Apollo/Sets/Terrain/Textures/T_Apollo_Grass_D_01
  }
  n6 = TextureSampleParameter2D(UVs=@n2) {
   group=Ground Mobile
   parameterName=Ground Diffuse Mobiile 2
   sortPriority=2
   texture=&T_Apollo_Grass_D_01
  }
  n7 = MaterialFunctionCall(A=@n5.RGB, B=@n6.RGB, Transition Phase=VertexColor().R, Height Texture=@n6.RGB, Contrast=Constant() {r=0.5}) {materialFunction=&HeightLerp}
  NamedRerouteDeclaration(None=LinearInterpolate(A=@n3.RGB, B=@n4.RGB, Alpha=@n7.Alpha)) {name=Mobile Spec, nodeColor=<rgba 0.125 0 1 1>}
  n8 = ComponentMask(None=NamedRerouteUsage()) {b=true}
  n9 = TextureSampleParameter2D(UVs=TextureCoordinate() {coordinateIndex=1}) {
   group=Baked
   parameterName=Mask
   samplerType=SAMPLERTYPE_Masks
   sortPriority=1
   texture=/Game/Environments/Artemis/Sets/Milkshake/Terrain/Textures/T_Milkshake_Hole_MCAO
  }
  NamedRerouteDeclaration(None=StaticSwitchParameter(True=Max(A=@n8) {constB=0.15}, False=@n9.B) {group=Baked, parameterName=SingleRock}) {name=AO Mask/ Up Vector, nodeColor=<rgba 1 0.1406 0 1>}
  NamedRerouteDeclaration(None=@n7.Results) {name=Mobile Diffuse, nodeColor=<rgba 0 0.2031 1 1>}
  n10 = NamedRerouteUsage()
  n11 = Multiply(
   A=TextureCoordinate() {coordinateIndex=1}
   B=LinearInterpolate(
    A=ScalarParameter() {defaultValue=12, group=Ground, parameterName=Ground Tile, sortPriority=0}
    B=ScalarParameter() {defaultValue=24.1032, group=Ground, parameterName=Ground Tile Bottom, sortPriority=1}
    Alpha=Round(None=@n10)
   )
  )
  n12 = TextureSampleParameter2D(UVs=@n11) {group=Ground, parameterName=Ground Diffuse, sortPriority=2, texture=/Game/Environments/Apollo/Sets/Terrain/Textures/T_Apollo_Dirt_D_01}
  NamedRerouteDeclaration(None=@n12.A) {name=Ground A, nodeColor=<rgba 0.3359 0 1 1>}
  NamedRerouteDeclaration(None=@n9.R) {name=Top Mask, nodeColor=<rgba 0.3359 1 0 1>}
  NamedRerouteDeclaration(None=@n9.G) {name=Curvature, nodeColor=<rgba 0 1 0.7422 1>}
  PinBase()
  n13 = Subtract(A=WorldPosition().XYZ, B=ActorPositionWS())
  n14 = ScalarParameter() {defaultValue=-546.6205, group=Ground, parameterName=Bottom _HeightOffset, sortPriority=6}
  n15 = ComponentMask(None=Divide(A=Subtract(A=@n13, B=@n14), B=ScalarParameter() {defaultValue=-183.6973, group=Ground, parameterName=Bottom_Blur, sortPriority=7})) {b=true}
  n16 = VertexNormalWS()
  n17 = Saturate(None=Multiply(A=@n15, B=ComponentMask(None=@n16) {b=true}))
  n18 = Saturate(None=@n15)
  PinBase!(Fade *Up=@n17, Fade Only=@n18) {reroutePins [
    Fade *Up {expression=&self:MaterialExpressionReroute_2}
    Fade Only {expression=&self:MaterialExpressionReroute_0}
   ]}
  PinBase!() {reroutePins [HLOD_Offset {expression=&self:MaterialExpressionReroute_3}]}
  n19 = Subtract(A=WorldPosition().XYZ, B=ActorPositionWS())
  MaterialExpressionReroute_3 = Reroute(None=@n1.RGB)
  n20 = MaterialProxyReplace(Realtime=@n19, MaterialProxy=Subtract(A=@n19, B=@MaterialExpressionReroute_3.HLOD_Offset) {desc=Account for HLOD Baking Offset})
  n21 = ScalarParameter() {defaultValue=-546.6205, group=Ground, parameterName=Bottom _HeightOffset, sortPriority=6}
  n22 = ComponentMask(None=Divide(A=Subtract(A=@n20, B=@n21), B=ScalarParameter() {defaultValue=-183.6973, group=Ground, parameterName=Bottom_Blur, sortPriority=7})) {b=true}
  n23 = ComponentMask(None=VertexNormalWS()) {b=true}
  PinBase!(Fade *Up=Saturate(None=Multiply(A=@n22, B=@n23)), Fade Only=Saturate(None=@n22)) {reroutePins [
    Fade *Up {expression=&self:MaterialGraph_0.MaterialGraphNode_Composite_3.Height Masks.MaterialExpressionReroute_2}
    Fade Only {expression=&self:MaterialGraph_0.MaterialGraphNode_Composite_3.Height Masks.MaterialExpressionReroute_3}
   ]}
  PinBase!()
  n24 = WorldPosition()
  n25 = ActorPositionWS()
  n26 = Divide(
   A=Subtract(A=Subtract(A=@n24.XYZ, B=@n25), B=ScalarParameter() {defaultValue=-742.5847, group=Wet, parameterName=Wet Height})
   B=ScalarParameter() {defaultValue=608.915, group=Wet, parameterName=Wet Blur}
  )
  PinBase!(Output=Saturate(None=ComponentMask(None=@n26) {b=true})) {reroutePins [Output {expression=&self:MaterialGraph_0.MaterialGraphNode_Composite_2.Wet Height.MaterialExpressionReroute_0}]}
  n27 = RuntimeVirtualTextureSampleParameter() {
   group=RVT
   materialType=BaseColor_Normal_Specular_YCoCg
   parameterName=RVT
   sortPriority=3
   virtualTexture=/Game/Athena/Apollo/Maps/Landscape/VT/RVT_Apollo_Landscape
  }
  n28 = SetMaterialAttributes!(MaterialAttributes, Base Color=@n27.BaseColor, Specular=@n27.BaseColor, Roughness=@n27.BaseColor, Normal=@n27.BaseColor) {attributeSetTypes [
    #1
    #2
    #3
    #4
   ]}
  n29 = TextureSample(UVs=Multiply(A=TextureCoordinate() {coordinateIndex=1}, B=ScalarParameter() {defaultValue=15, group=Baked, parameterName=Details Tile, sortPriority=0})) {texture=/Game/Environments/Apollo/Rocks/Natural_Rocks/Textures/T_RockDetails_02D}
  n30 = MaterialFunctionCall(
   Base=LinearInterpolate(
    A=VectorParameter() {group=Baked, parameterName=Color Sides, sortPriority=-1, defaultValue=<rgba 0.408 0.397 0.386 0>}.RGB
    B=VectorParameter() {group=Baked, parameterName=Color Top, sortPriority=-1, defaultValue=<rgba 0.1267 0.0944 0.0464 0>}.RGB
    Alpha=NamedRerouteUsage()
   )
   Blend=@n29.RGB
  ) {materialFunction=&Blend_Overlay}
  n31 = NamedRerouteUsage()
  n32 = NamedRerouteUsage()
  n33 = Constant() {r=1}
  n34 = Constant() {r=-0.1333}
  n35 = TextureSampleParameter2D() {
   group=Baked
   parameterName=Normals
   samplerSource=SSM_Clamp_WorldGroupSettings
   samplerType=SAMPLERTYPE_Normal
   sortPriority=2
   texture=/Game/Environments/Artemis/Sets/Milkshake/Terrain/Textures/T_Milkshake_Hole_Sides_N
  }
  n36 = NamedRerouteDeclaration(None=@n35.RGB) {name=Baked N, nodeColor=<rgba 0 1 0.8594 1>}
  n37 = SetMaterialAttributes(
   Base Color=NamedRerouteDeclaration(None=MaterialFunctionCall(Base=@n30.Result, Blend=@n31) {materialFunction=&Blend_Overlay}.Result) {name=Backed Diffuse, nodeColor=<rgba 1 0 0.0469 1>}
   Specular=LinearInterpolate(A=Constant() {r=-0.746}, B=Constant() {r=0.5427}, Alpha=@n32)
   Roughness=Add(A=LinearInterpolate(A=@n33, B=@n34, Alpha=@n32), B=@n29.R)
   Normal=NamedRerouteDeclaration(None=Transform(None=@n36)) {name=Baked N WS, nodeColor=<rgba 1 0.2578 0 1>}
  ) {attributeSetTypes [
    #1
    #2
    #3
    #4
   ]}
  n38 = NamedRerouteDeclaration(None=@n37) {name=Baked Only, nodeColor=<rgba 0.5469 0 1 1>}
  n39 = VectorParameter() {group=Ground Mobile, parameterName=GroundTone Top Mobile, sortPriority=-2, defaultValue=<rgba 1 1 1 0>}
  n40 = VectorParameter() {group=Ground, parameterName=GroundTone Top, sortPriority=-2, defaultValue=<rgba 0.6609 0.7448 0.666 0>}
  n41 = StaticSwitchParameter(True=@n39.RGB, False=ShadingPathSwitch(Default=@n40.RGB, Mobile=@n39.RGB)) {group=Ground Mobile, parameterName=Debug Mobile}
  n42 = VectorParameter() {group=Ground, parameterName=GroundToneBottom, sortPriority=-1, defaultValue=<rgba 0.526 0.8819 1 0>}
  n43 = NamedRerouteUsage()
  n44 = NamedRerouteUsage()
  n45 = LinearInterpolate(A=NamedRerouteUsage(), B=@n12.RGB, Alpha=@n44)
  n46 = ShadingPathSwitch(Default=@n12.RGB, Mobile=@n45)
  n47 = LinearInterpolate(A=NamedRerouteUsage(), Alpha=@n44) {constB=0.5}
  n48 = TextureSampleParameter2D(UVs=@n11) {
   group=Ground
   parameterName=Ground Spec
   samplerType=SAMPLERTYPE_Masks
   sortPriority=3
   texture=/Game/Environments/Apollo/Sets/Terrain/Textures/T_Apollo_Dirt_S_01
  }
  n49 = StaticSwitchParameter(True=@n47, False=ShadingPathSwitch(Default=@n48.RGB, Mobile=@n47)) {group=Ground Mobile, parameterName=Debug Mobile}
  n50 = NamedRerouteUsage()
  n51 = TextureSampleParameter2D(UVs=@n11) {
   group=Ground
   parameterName=Ground Normal
   samplerType=SAMPLERTYPE_Normal
   sortPriority=4
   texture=/Game/Environments/Apollo/Sets/Terrain/Textures/T_Apollo_Dirt2_N
  }
  n52 = Multiply(A=@n51.RGB, B=ScalarParameter() {defaultValue=1.5, group=Ground, parameterName=Ground Normals Value, sortPriority=5})
  n53 = Constant3Vector() {constant=<rgba 1 1 0 0>}
  n54 = ShadingPathSwitch(Default=Add(A=@n50, B=Multiply(A=@n52, B=@n53)), Mobile=@n50)
  n55 = SetMaterialAttributes(
   Base Color=Multiply(
    A=LinearInterpolate(A=@n41, B=@n42.RGB, Alpha=@n43) {constA=1}
    B=StaticSwitchParameter(True=@n45, False=@n46) {group=Ground Mobile, parameterName=Debug Mobile}
   )
   Specular=ComponentMask(None=@n49) {r=true}
   Roughness=ComponentMask(None=@n49) {b=true}
   Normal=Transform(None=StaticSwitchParameter(True=@n50, False=@n54) {group=Ground Mobile, parameterName=Debug Mobile})
  ) {attributeSetTypes [
    #1
    #2
    #3
    #4
   ]}
  n56 = NamedRerouteUsage()
  MaterialExpressionReroute_2 = Reroute(None=@n17)
  n57 = StaticSwitchParameter(
   True=NamedRerouteUsage()
   False=BlendMaterialAttributes(
    A=BlendMaterialAttributes(A=@n38, B=@n55, Alpha=@n56)
    B=@n55
    Alpha=NamedRerouteDeclaration(None=@MaterialExpressionReroute_2.Fade *Up) {name=Bottom Mask, nodeColor=<rgba 1 0 0.9375 1>}
   )
  ) {group=Baked, parameterName=SingleRock}
  n58 = GetMaterialAttributes(None=@n57) {attributeGetTypes [#1, #3]}
  n59 = SetMaterialAttributes!(
   MaterialAttributes=@n58.MaterialAttributes
   Base Color=Multiply(A=@n58.Base Color, B=ScalarParameter() {defaultValue=0.35, group=Wet, parameterName=Wet Tone Scale, sortPriority=1}) {constB=0.35}
   Specular=ScalarParameter() {defaultValue=0.25, group=Wet, parameterName=WeT Specular, sortPriority=2}
   Roughness=Multiply(A=@n58.Roughness, B=ScalarParameter() {defaultValue=0.34, group=Wet, parameterName=Wet Roughness Scale, sortPriority=3}) {constB=0.2}
  ) {attributeSetTypes [#1, #2, #3]}
  MaterialExpressionReroute_0 = Reroute(None=@n18)
  n60 = ShadingPathSwitch(
   Default=StaticSwitchParameter(True=@n57, False=BlendMaterialAttributes(A=@n59, B=@n58.MaterialAttributes, Alpha=@MaterialExpressionReroute_0.Output)) {group=Wet, parameterName=No Wet, sortPriority=0}
   Forward=@n57
   Mobile=@n57
  )
  n61 = NamedRerouteDeclaration(None=@n60) {name=Pre RVT, nodeColor=<rgba 1 0 0.75 1>}
  n62 = ScalarParameter() {defaultValue=100, group=RVT, parameterName=Blend RVT, sortPriority=0}
  n63 = ScalarParameter() {defaultValue=20.5044, group=RVT, parameterName=Blend RVT Bottom, sortPriority=0}
  n64 = RuntimeVirtualTextureSampleParameter() {
   group=RVT
   materialType=WorldHeight
   parameterName=RVTHeight
   sortPriority=4
   virtualTexture=/Game/Athena/Apollo/Maps/Landscape/VT/RVT_Apollo_Landscape_Height
  }
  n65 = Subtract(A=ComponentMask(None=WorldPosition().XYZ) {b=true}, B=@n64.WorldHeight)
  n66 = ScalarParameter() {defaultValue=3, desc=Tiled Diffuse Alpha is used here, group=RVT, parameterName=RVT HeightMap Contrast, sortPriority=2}
  n67 = Multiply(A=@n66) {constB=-1}
  n68 = NamedRerouteUsage()
  n69 = MaterialFunctionCall(
   Transition Phase=Divide(
    A=NamedRerouteDeclaration(None=@n65) {name=RVT Height, nodeColor=<rgba 1 0 0.6562 1>}
    B=LinearInterpolate(A=@n62, B=@n63, Alpha=@MaterialExpressionReroute_0.Fade Only)
   )
   Height Texture=LinearInterpolate(A=LinearInterpolate(A=@n67, B=@n66, Alpha=@n68), B=@n68, Alpha=@MaterialExpressionReroute_0.Fade Only)
   Contrast=LinearInterpolate(
    A=ScalarParameter() {defaultValue=0.25, group=RVT, parameterName=Blend RVT Contrast, sortPriority=1}
    B=ScalarParameter() {group=RVT, parameterName=Blend RVT Contrast Bottom, sortPriority=1}
    Alpha=@MaterialExpressionReroute_0.Fade Only
   )
  ) {materialFunction=&HeightLerp}
  n70 = MaterialFunctionCall(
   NoRVT=@n60
   RVT=MaterialProxyReplace(Realtime=BlendMaterialAttributes(A=@n28, B=@n61, Alpha=@n69.Alpha) {vertexAttributeBlendType=UseA}, MaterialProxy=@n60)
  ) {materialFunction=&FN_RVT_Switch_Material}
  StaticSwitchParameter(
   True=@n60
   False=StaticSwitchParameter(True=@n60, False=ShadingPathSwitch(Default=@n70.Result, Deferred=@n70.Result, Forward=@n60, Mobile=@n60)) {group=RVT, parameterName=Skip RVT, sortPriority=-4}
  ) {group=Ground Mobile, parameterName=Debug Mobile}
 ]
}
